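# Translucent unlit Niagara ribbon material for the Rain FNCS pickaxe's patterned trail.
# Color comes from a banded gradient broken up by camera-distance-scaled Ben-Day dots, mapped through color parameters or a curve atlas row.
# Scrolled-free wispy and block noise textures add highlights, hard-edged color patches and ragged edges.
# Dynamic parameters drive state blends, variant colors and trailing-edge erosion; opacity combines edge fades, particle alpha and near-camera fade.
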Material /BRCosmetics/Weapons/FORT_Melee/Pickaxe_RainFNCS/FX/M_RainFNCS_Pickaxe_PatternTrail
graph {
 properties {
  bAutomaticallySetUsageInEditor=false
  bUsedWithNiagaraRibbons=true
  blendMode=BLEND_Translucent
  shadingModel=MSM_Unlit
  usedShadingModels=Unlit
 }
 nodes [
  n1 = VectorParameter() {parameterName=Color_1, defaultValue=<rgba 0 0.0131 1 1>}
  n2 = Power(Base=ComponentMask(None=TextureCoordinate()) {g=true}) {constExponent=1.7}
  n3 = LinearInterpolate(
   A=@n1.RGB
   B=LinearInterpolate(A=@n1.RGB, B=VectorParameter() {parameterName=Color_9, defaultValue=<rgba 0 0.0131 1 1>}.RGB, Alpha=Saturate(None=@n2))
   Alpha=ScalarParameter() {parameterName=LerpDotsGradient}
  )
  n4 = VectorParameter() {parameterName=Color_2, defaultValue=<rgba 0.0013 1 0.6247 1>}
  n5 = ComponentMask(None=TextureCoordinate()) {g=true}
  n6 = Round(None=ScalarParameter() {defaultValue=5, parameterName=Bands})
  n7 = TextureCoordinate() {coordinateIndex=1}
  n8 = AppendVector(A=ScalarParameter() {defaultValue=10, parameterName=Dots Tiling Multiplier}, B=Constant() {r=1})
  n9 = ScalarParameter() {defaultValue=30, group="Ben Day Dots [Old School Cartoon]", parameterName=DotsAmount, sliderMax=512, sortPriority=20}
  n10 = Subtract(A=Distance(A=ActorPositionWS(), B=CameraPositionWS())) {constB=100}
  n11 = Multiply(A=@n9, B=Saturate(None=Divide(A=@n10) {constB=400})) {constB=0.5}
  n12 = Rotator(Coordinate=Multiply(A=Multiply(A=@n7, B=@n8), B=LinearInterpolate(A=@n9, B=@n11)), Time=Constant() {r=1}) {speed=0.75}
  n13 = Distance(A=Add(A=Frac(None=@n12)) {constB=-0.5}, B=Constant())
  n14 = ScalarParameter() {defaultValue=2, parameterName=Dots Range}
  n15 = NamedRerouteUsage()
  n16 = Add(A=Multiply(A=@n5, B=@n6), B=Multiply(A=Multiply(A=@n13, B=@n14) {constB=2.5}, B=@n15))
  n17 = LinearInterpolate(A=@n5, B=Divide(A=Round(None=@n16), B=@n6), Alpha=@n15) {constAlpha=1}
  n18 = VectorParameter() {parameterName=Color_6, defaultValue=<rgba 0 0.0131 1 1>}
  n19 = VectorParameter() {parameterName=Color_7, defaultValue=<rgba 0.0013 1 0.6247 1>}
  n20 = DynamicParameter() {parameterIndex=1, defaultValue=<rgba 0 0 0 1>, paramNames [VariantColors, Param2, Param3, Param4]}
  n21 = CurveAtlasRowParameter(CurveTime=@n17) {
   atlas=&/CA_RainFNCS_Gradients
   curve=&/C_RainFNCS_Maincolor
  }
  n22 = LinearInterpolate(
   A=LinearInterpolate(A=LinearInterpolate(A=@n3, B=@n4.RGB, Alpha=@n17), B=LinearInterpolate(A=@n18.RGB, B=@n19.RGB, Alpha=@n17), Alpha=@n20.VariantColors)
   B=@n21
   Alpha=ScalarParameter() {defaultValue=1, parameterName=UseColorCurve}
  )
  n23 = Multiply(
   A=MaterialFunctionCall(UVs=TextureCoordinate() {coordinateIndex=1}, Rotation Center=Constant2Vector() {g=0.5, r=0.5}, "Rotation Angle (0-1)"=Constant() {r=-0.125}) {materialFunction=&CustomRotator}.Rotated Values
   B=Constant2Vector() {g=0.25, r=1}
  ) {constB=2}
  n24 = Multiply(A=@n23) {constB=2}
  n25 = LinearInterpolate(
   A=@n22
   B=Desaturation(None=Multiply(A=@n22) {constB=0.75}, Fraction=ScalarParameter() {defaultValue=-0.5, parameterName=Desaturation})
   Alpha=LinearInterpolate(A=TextureSample(UVs=@n24) {texture=/BRCosmetics/Athena/Effects/Particles/Skydiving/NS_Contrail_Chainmail/T_WispyFireNoise03}.R, Alpha=@n15) {constB=0}
  )
  n26 = VectorParameter() {parameterName=Color_3, defaultValue=<rgba 1 0.1502 0 1>}
  n27 = VectorParameter() {parameterName=Color_8, defaultValue=<rgba 1 0.1502 0 1>}
  n28 = DynamicParameter() {parameterIndex=1, defaultValue=<rgba 0 0 0 1>, paramNames [VariantColors, Param2, Param3, Param4]}
  n29 = TextureSample(UVs=Multiply(A=@n23) {constB=3}) {texture=&T_WispyFireNoise03}
  n30 = TextureCoordinate() {coordinateIndex=1}
  n31 = Constant2Vector() {g=0.66, r=0.33}
  n32 = TextureSample(UVs=MaterialFunctionCall(UVs=Add(A=@n30, B=@n31), Rotation Center=Constant2Vector() {g=0.5, r=0.5}, "Rotation Angle (0-1)"=Constant() {r=-0.125}) {materialFunction=&CustomRotator}.Rotated Values) {samplerType=SAMPLERTYPE_LinearGrayscale, texture=&/T_Blocknoise_M}
  n33 = Multiply(A=OneMinus(None=@n32.R)) {constB=0.3}
  n34 = Power(Base=@n32.R) {constExponent=0.5}
  n35 = OneMinus(None=ComponentMask(None=TextureCoordinate()) {g=true})
  n36 = Saturate(None=Subtract(A=@n34, B=Multiply(A=@n35) {constB=0.75}))
  n37 = Multiply(A=OneMinus(None=Multiply(A=@n29.R)), B=Power(Base=Distance(A=@n33, B=@n36)) {constExponent=0.5})
  n38 = LinearInterpolate(A=LinearInterpolate(A=@n25, B=LinearInterpolate(A=@n26.RGB, B=@n27.RGB, Alpha=@n28.VariantColors), Alpha=@n37), B=@n25, Alpha=@n15)
  n39 = LinearInterpolate(Alpha=NamedRerouteUsage()) {constA=0.45}
  n40 = Multiply(
   A=SmoothStep(Min=@n39, Max=Add(A=@n39) {constB=0.02}, Value=@n36) {constMax=0.5, constMin=0.5}
   B=ScalarParameter() {defaultValue=1, parameterName=HardEdgeMaskMult}
  )
  n41 = DynamicParameter() {parameterIndex=1, defaultValue=<rgba 0 0 0 1>, paramNames [VariantColors, Param2, Param3, Param4]}
  n42 = LinearInterpolate(
   A=LinearInterpolate(A=VectorParameter() {parameterName=Color_5, defaultValue=<rgba 0 0 0 1>}.RGB, Alpha=@n41.VariantColors) {constB=0}
   B=LinearInterpolate(A=VectorParameter() {parameterName=Color_4, defaultValue=<rgba 0.2952 0.0313 0.477 1>}.RGB, Alpha=@n41.VariantColors) {constB=0}
   Alpha=OneMinus(None=Power(Base=@n32.R) {constExponent=1})
  )
  n43 = ParticleColor()
  MaterialFunctionCall(Emissive=Multiply(A=LinearInterpolate(A=@n38, B=@n42, Alpha=@n40), B=@n43.RGB)) {materialFunction=&MF_FX_EmissiveAdjustments}
  n44 = ComponentMask(None=TextureCoordinate()) {r=true}
  n45 = ScalarParameter() {group="Fade Leading && Trailing Edges", parameterName=Leading Edge Fade Bias, sortPriority=2}
  n46 = Saturate(None=Multiply(A=Add(A=@n44, B=@n45), B=ScalarParameter() {defaultValue=5, group="Fade Leading && Trailing Edges", parameterName=Leading Edge Fade Scale}) {constB=3})
  n47 = ScalarParameter() {defaultValue=1, group=Alpha, parameterName=Fade Leading Edge, sortPriority=4}
  n48 = Subtract(
   A=OneMinus(None=@n44)
   B=Multiply(A=@n37, B=DynamicParameter() {defaultValue=<rgba 0 0 0 1>, paramNames [StateMaskBlend, StateOpacityBlend, Erode, Param4]}.Erode)
  )
  n49 = ScalarParameter() {group="Fade Leading && Trailing Edges", parameterName=Trailing Edge Fade Bias, sortPriority=3}
  n50 = Saturate(None=Multiply(
    A=Add(A=@n48, B=@n49)
    B=ScalarParameter() {defaultValue=5, group="Fade Leading && Trailing Edges", parameterName=Trailing Edge Fade Scale, sortPriority=1}
   ) {constB=2})
  n51 = ScalarParameter() {defaultValue=1, group=Alpha, parameterName=Fade Trailing Edge, sortPriority=5}
  n52 = LinearInterpolate(A=@n37, B=Constant() {r=1}, Alpha=LinearInterpolate(Alpha=NamedRerouteUsage()) {constA=0.5}) {constAlpha=0}
  n53 = ComponentMask(None=TextureCoordinate()) {g=true}
  n54 = Multiply(A=@n37) {constB=0.05}
  n55 = Subtract(A=@n53, B=@n54)
  n56 = ScalarParameter() {defaultValue=5, group="Fade Top && Bottom Length Edges", parameterName=Top Length Fade Scale}
  n57 = LinearInterpolate(
   B=Saturate(None=Multiply(A=@n55, B=@n56) {constB=3})
   Alpha=ScalarParameter() {defaultValue=1, group=Alpha, parameterName=Fade Top Length Edge, sortPriority=7}
  ) {constA=1}
  n58 = Multiply(A=OneMinus(None=@n34)) {constB=0.25}
  n59 = OneMinus(None=@n53)
  n60 = Multiply(A=Add(A=Subtract(A=@n58, B=@n59)) {constB=0.1}) {constB=15}
  n61 = Subtract(A=@n59, B=@n54)
  n62 = LinearInterpolate(A=Saturate(None=OneMinus(None=@n60)), B=Saturate(None=Multiply(A=@n61) {constB=5}), Alpha=NamedRerouteUsage())
  n63 = Multiply(
   A=Multiply(A=LinearInterpolate(B=@n46, Alpha=@n47) {constA=1}, B=LinearInterpolate(B=@n50, Alpha=@n51) {constA=1})
   B=Multiply(A=Add(A=@n52, B=@n40), B=Multiply(A=@n57, B=@n62))
  )
  Multiply(
   A=Multiply(A=@n43.A, B=Saturate(None=@n63))
   B=MaterialFunctionCall(Distance=ScalarParameter() {defaultValue=32, group=Alpha, parameterName=NearFadeDistance, sortPriority=3}) {materialFunction=&Near_Camera_Fade}.Result
  )
  n64 = DynamicParameter() {defaultValue=<rgba 0 0 0 1>, paramNames [StateMaskBlend, StateOpacityBlend, Erode, Param4]}
  NamedRerouteDeclaration(None=@n64.StateMaskBlend) {name=StateMaskBlend, nodeColor=<rgba 0.9453 0 1 1>}
  NamedRerouteDeclaration(None=@n64.StateOpacityBlend) {name=StateMaskBlendOpacity, nodeColor=<rgba 1 0.3047 0 1>}
  TextureSample(UVs=@n23) {texture=&T_WispyFireNoise03}
 ]
}
