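# Niagara idle effect for the RainFNCS pickaxe: a persistent capsule-shaped glow built from sprites.
# The CapsuleGlow emitter bursts System.ParticleAmount sprites, using each one's normalized partition to drive curves for position along a slightly bent XZ line, size and color.
# Color lerps by System.StateBlend between a purple-orange-yellow gradient and an orange-yellow one, with alpha fading in as the system starts.
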NiagaraSystem /BRCosmetics/Weapons/FORT_Melee/Pickaxe_RainFNCS/FX/NS_Pickaxe_RainFNCS_Idle
dependencies {
 usedDataInterfaces [&NiagaraDataInterfaceCurve, &NiagaraDataInterfaceColorCurve]
 usedDynamicInputs [&MakeVectorFromLinearColorRGB, /Niagara/DynamicInputs/Lerp/SmoothLerpOverTime_Color]
 usedModules [
  &self:SystemSpawnScript.SystemScriptSource.SystemScriptGraph.NiagaraNodeAssignment_1.SetVariables_#1
  &SystemState
  &self:SystemSpawnScript.SystemScriptSource.SystemScriptGraph.NiagaraNodeAssignment_2.SetVariables_#2
 ]
}
emitters {CapsuleGlow {
  emitterUpdateScript [
   &EmitterState(
    Inactive Response=::NewEnumerator1
    Life Cycle Mode=::NewEnumerator1
    Loop Behavior=::NewEnumerator1
    Loop Duration Mode=::NewEnumerator1
    UseLoopDelay=false
   )
   &SpawnBurst_Instantaneous(Spawn Count=$System.ParticleAmount, Use Spawn Probability=false, UseLoopCountLimit=false)
  ]
  particleSpawnScript [
   &PartitionParticles(Number Of Partitions=$System.ParticleAmount)
   &InitializeParticle(
    Color=<rgba 0.5 0.5 0.5 1>
    Sprite Rotation Mode=::NewEnumerator0
    Sprite Size=&Multiply_Vector2D_ByFloat(
     Float=&FloatFromCurve(
      FloatCurve=DataInterface() {propertyValues {Curve {
         defaultValue=1
         keys [
          {arriveTangent=1.6083, interpMode=RCIM_Cubic, leaveTangent=1.6083, tangentMode=RCTM_User, time=0, value=0.7388}
          {arriveTangent=-0.3844, interpMode=RCIM_Cubic, leaveTangent=-0.3844, tangentMode=RCTM_User, time=0.5, value=0.9585}
          {arriveTangent=-1.75, interpMode=RCIM_Cubic, leaveTangent=-1.75, tangentMode=RCTM_User, time=1, value=0.6}
         ]
        }}}
      CurveIndex=$Particles.PartitionParticles.NormalizedPartition
     )
     Vector2D=<13 13>
    )
    Sprite Size Mode=::NewEnumerator4
    UsePositionOffset=true
    Write Material Random=true
   )
   &CameraOffset(Camera Offset Amount=-4)
   /SetVariables(Particles.Position=&MakeVector(
     X=&FloatFromCurve(
      FloatCurve=DataInterface() {propertyValues {Curve {
         defaultValue=1
         keys [<at 0 0 RCIM_Cubic>, {arriveTangent=20.0632, interpMode=RCIM_Cubic, leaveTangent=20.0632, tangentMode=RCTM_User, time=1, value=4.7205}]
        }}}
      CurveIndex=$Particles.PartitionParticles.NormalizedPartition
     )
     Y=0
     Z=&FloatFromCurve(
      FloatCurve=DataInterface() {propertyValues {Curve {defaultValue=1, keys [{tangentMode=RCTM_User, time=0, value=12.2949}, {tangentMode=RCTM_User, time=1, value=54.8911}]}}}
      CurveIndex=$Particles.PartitionParticles.NormalizedPartition
     )
    )) {isSetParameters=true}
  ]
  particleUpdateScript [
   &ParticleState(Kill Particles When Lifetime Has Elapsed=false, Let Infinitely Lived Particles Die When Emitter Deactivates=true)
   &ScaleColor(
    Scale Mode=::NewEnumerator1
    Scale RGBA=&MakeVector4_FromLinearColor(LinearColor=&LerpLinearColors(
      StartColor=&ColorFromCurve(
       ColorCurve=DataInterface() {propertyValues {
         AlphaCurve {keys [<at 0 0.25>,<at 0.7997 0.25>,<at 1 0>]}
         BlueCurve {keys [<at 0.12 0.4583>,<at 0.5545 0>,<at 0.9085 0>]}
         GreenCurve {keys [<at 0.12 0>,<at 0.5545 0.1559>,<at 0.9085 0.7397>]}
         RedCurve {keys [<at 0.12 0.2025>,<at 0.5545 1>,<at 0.9085 1>]}
        }}
       CurveIndex=$Particles.PartitionParticles.NormalizedPartition
      )
      EndColor=&ColorFromCurve(
       ColorCurve=DataInterface() {propertyValues {
         AlphaCurve {keys [<at 0 0.25>,<at 0.7997 0.25>,<at 1 0>]}
         BlueCurve {keys [<at 0.1192 0>,<at 0.9085 0>]}
         GreenCurve {keys [<at 0.1192 0.1559>,<at 0.9085 0.7397>]}
         RedCurve {keys [<at 0.1192 1>,<at 0.9085 1>]}
        }}
       CurveIndex=$Particles.PartitionParticles.NormalizedPartition
      )
      LerpFactor=$System.StateBlend
     ))
    ScaleRGBA=true
   )
   &ScaleColor(
    Scale Alpha=&FloatFromCurve(
     FloatCurve=DataInterface() {propertyValues {Curve {defaultValue=1, keys [<at 0 0 RCIM_Cubic>,<at 0.2 1 RCIM_Cubic>]}}}
     CurveIndex=$Engine.System.Age
    )
    ScaleA=true
    ScaleRGB=true
   ) {module=ScaleColor001}
  ]
  rendererClasses [&NiagaraSpriteRendererProperties]
  renderers [{rendererClass=&NiagaraSpriteRendererProperties}]
 }}
summary {
 systemName=NS_Pickaxe_RainFNCS_Idle
 emitters [{emitterName=CapsuleGlow, rendererClasses [&NiagaraSpriteRendererProperties]}]
 userVariables [
  User.bIsFrontEnd {type=NiagaraBool}
  User.bIsFrontEndPreview {type=NiagaraBool}
 ]
}
