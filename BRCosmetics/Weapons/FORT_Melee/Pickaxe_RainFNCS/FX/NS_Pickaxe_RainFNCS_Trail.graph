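# Niagara trail effect for the RainFNCS pickaxe: an animation-driven ribbon trail plus glitchy shapes.
# Trail draws two ribbons with FN AnimTrail modules, fading colour over adjusted age and passing state blend and opacity into particle values.
# GlitchShapes spawns per unit moved from socket-sampled locations, rendering short-lived, randomly sized meshes and sprites with waveform jitter in dynamic material parameters.
# A second, shorter-lived trail emitter, Trail001, is disabled.
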NiagaraSystem /BRCosmetics/Weapons/FORT_Melee/Pickaxe_RainFNCS/FX/NS_Pickaxe_RainFNCS_Trail
dependencies {
 usedDataInterfaces [
  &NiagaraDataInterfaceColorCurve
  &NiagaraDataInterfaceCurve
  &NiagaraDataInterfaceArrayInt32
  &NiagaraDataInterfaceArrayFloat
  &NiagaraDataInterfaceSocketReader
 ]
 usedDynamicInputs [&FNTimeOfDayColor]
 usedModules [
  &self:SystemSpawnScript.SystemScriptSource.SystemScriptGraph.NiagaraNodeAssignment_1.SetVariables_#1
  &self:SystemSpawnScript.SystemScriptSource.SystemScriptGraph.NiagaraNodeAssignment_0.SetVariables_#2
  &SystemState
  &self:SystemSpawnScript.SystemScriptSource.SystemScriptGraph.NiagaraNodeAssignment_2.SetVariables_#3
 ]
}
emitters {
 GlitchShapes {
  emitterUpdateScript [
   &EmitterState()
   &SpawnPerUnit(Module.Use Max Movement Threshold=true, Module.Use Movement Tolerance=false, Use Spawn Probability=false)
  ]
  particleSpawnScript [
   /SetVariables(Particles.RandomBone:NiagaraInt32=&RandomRangeInt(Minimum=0, Maximum=5)) {isSetParameters=true}
   /SetVariables(Particles.SizeLerp:NiagaraFloat=&FloatFromCurve(
     FloatCurve=DataInterface() {propertyValues {Curve {keys [
         {interpMode=RCIM_Cubic, tangentMode=RCTM_User, time=0, value=0}
         {arriveTangent=1.4324, interpMode=RCIM_Cubic, leaveTangent=1.4324, tangentMode=RCTM_User, time=0.87, value=0.5596}
         {interpMode=RCIM_Cubic, tangentMode=RCTM_User, time=1, value=1}
        ]}}}
     CurveIndex=&RandomRangeFloat(Minimum=0, Maximum=1)
    )) {isSetParameters=true}
   /SetVariables(Particles.BoneIndex:NiagaraInt32=&MaskIntBySpawnGroup(
     Masked Int=&RandomRangeInt(Minimum=0, Maximum=1)
     Passthrough Int=&SelectIntFromArray(Direct Array Index=$Particles.RandomBone, Int Selection Array=DataInterface() {propertyValues {IntData [0,1,5,6,9,10]}})
     Spawn Group Mask=1
    )) {isSetParameters=true}
   &InitializeParticle(
    Color=&RandomRangeLinearColor(Minimum=<rgba 0.9045 0.2554 2 1>, Maximum=<rgba 4 2.9411 0.0098 1>)
    Lifetime Max=0.3
    Lifetime Min=0.15
    Lifetime Mode=::NewEnumerator1
    Mesh Scale=&Multiply_VectorByFloat(
     Vector=&RandomRangeVector(Minimum=<0.06 0.06 0.06>, Maximum=<0.11 0.11 0.11>)
     Float=&SelectFloatFromArray(Direct Array Index=$Particles.RandomBone, Float Selection Array=DataInterface() {propertyValues {FloatData [1.2,1.2,0.8,0.8,0.5,0.5]}})
    )
    Mesh Scale Mode=::NewEnumerator4
    Sprite Size Mode=::NewEnumerator3
    UsePositionOffset=false
    Write Material Random=true
   )
   &InitialMeshOrientation(Use Rotation Vector=false)
   &CameraOffset(Camera Offset Amount=8)
   &SocketLocation(
    SamplePosOffset=true
    Sampled Position Offset=&RandomRangeVector(Minimum=<0 0 0>, Maximum=<0 1 30>)
    Sampled Velocity Coordinate Space:ENiagaraCoordinateSpace=::Simulation
    Socket Reader=$System.Socket Reader
    Socket Sampling Mode=::NewEnumerator3
    VelocitySpeedLimit=false
   )
  ]
  particleUpdateScript [
   &ParticleState()
   /SetVariables(Particles.Jitter:NiagaraFloat=&ClampFloat(
     Float=&Waveform(
      Global Clamp Min=-1
      "Waveform [1]"=::NewEnumerator12
      "[1] Frequency"=&RandomRangeFloat(Minimum=15, Maximum=25)
      "[1] Amplitude Min/Max"=<-150 150>
     )
     Min=-0.25
     Max=0.25
    )) {isSetParameters=true}
   &DynamicMaterialParameters(
    Index 0 Param 1=&Add_Float(A=$Particles.Jitter, B=&RandomRangeFloat(Minimum=0, Maximum=1))
    Index 0 Param 2=&Subtract_Float(A=&RandomRangeFloat(Minimum=0, Maximum=1), B=$Particles.Jitter)
    Index 0 Param 3=$Particles.Jitter
    Module.Param0WriteEnabled=true
    Module.Param1WriteEnabled=true
    Module.Param2WriteEnabled=true
    Module.Param3WriteEnabled=false
   )
   &ScaleColor(
    Scale Alpha=&FloatFromCurve(
     FloatCurve=DataInterface() {propertyValues {Curve {
        defaultValue=0
        keys [{interpMode=RCIM_Cubic, tangentMode=RCTM_User, time=0.8, value=1}, {interpMode=RCIM_Cubic, tangentMode=RCTM_User, time=1, value=0}]
       }}}
     Scale Curve=1.5
    )
    ScaleA=true
    ScaleRGB=true
   )
   &UpdateMeshOrientation(Rotation Rate=0.2, Rotation Vector=&RandomRangeVector(Minimum=<0 0 0>, Maximum=<1 1 1>))
   &Drag(Drag=10, Use Linear Drag=true, Use Rotational Drag=true)
   &InheritVelocity(Inherited Velocity Speed Limit=10, Speed Limit=true, UseSpeedThreshold=false)
   &SolveForcesAndVelocity(Clamp Velocity=false, Limit Acceleration=false)
  ]
  rendererClasses [&NiagaraMeshRendererProperties, &NiagaraSpriteRendererProperties]
  renderers [
   {rendererClass=&NiagaraMeshRendererProperties, rendererIndex=0}
   {rendererClass=&NiagaraSpriteRendererProperties, rendererIndex=1}
  ]
 }
 Trail {
  emitterSpawnScript [&FN_RibbonIDOffsetUtility()]
  emitterUpdateScript [
   &EmitterState()
   &FN_SpawnAnimTrail()
  ]
  particleSpawnScript [
   &FN_InitializeAnimTrailBoneIndecies()
   &FN_InitializeAnimTrail(Lifetime=0.39)
  ]
  particleUpdateScript [
   &FN_UpdateAnimTrail() {module=FN_UpdateAnimTrail001}
   /SetVariables(Particles.AdjustedNormalizedAge:NiagaraFloat=&Multiply_Float(A=$Particles.NormalizedAge, B=&Divide_Float(A=$Particles.Lifetime, B=0.18))) {isSetParameters=true}
   &SolveForcesAndVelocity(Clamp Velocity=false, Limit Acceleration=false)
   &ScaleColor(
    Scale Alpha=&FloatFromCurve(
     FloatCurve=DataInterface() {propertyValues {Curve {keys [{interpMode=RCIM_Cubic, tangentMode=RCTM_User, time=0, value=1}, {interpMode=RCIM_Cubic, tangentMode=RCTM_User, time=1, value=0}]}}}
     CurveIndex=$Particles.AdjustedNormalizedAge
     Scale Curve=2
    )
    Scale RGB=&MakeVectorFromLinearColorRGB(Color=&ColorFromCurve(
      ColorCurve=DataInterface() {propertyValues {
        AlphaCurve {keys [<at 0 1 RCIM_Cubic>,<at 1 0 RCIM_Cubic>]}
        BlueCurve {keys [<at 0 0 RCIM_Cubic>,<at 0.7877 0>]}
        GreenCurve {keys [<at 0 1.6464 RCIM_Cubic>,<at 0.7877 0.5279>]}
        RedCurve {keys [<at 0 3 RCIM_Cubic>,<at 0.7877 1.5>]}
       }}
      CurveIndex=$Particles.AdjustedNormalizedAge
     ))
    ScaleA=true
    ScaleRGB=true
   )
   /SetVariables(Particles.CleanTrailColor:LinearColor=&MakeLinearColorFromVectorAndFloat(
     "Vector (RGB)"=&VectorFromFloat(Value=&FloatFromCurve(
       FloatCurve=DataInterface() {propertyValues {Curve {defaultValue=1, keys [{interpMode=RCIM_Cubic, leaveTangent=-1.4397, tangentMode=RCTM_Break, time=0, value=1}]}}}
       CurveIndex=$Particles.AdjustedNormalizedAge
       Scale Curve=&Lerp_Float(A=0.75, B=1, Alpha=$System.StateBlend)
      ))
     "Float (Alpha)"=&FloatFromCurve(
      FloatCurve=DataInterface() {propertyValues {Curve {
         defaultValue=1
         keys [
          {arriveTangent=-0.0405, interpMode=RCIM_Cubic, leaveTangent=-0.2898, tangentMode=RCTM_Break, time=0, value=1}
          {arriveTangent=-2.2333, interpMode=RCIM_Cubic, leaveTangent=-2.2333, tangentMode=RCTM_User, time=1, value=0}
         ]
        }}}
      CurveIndex=$Particles.AdjustedNormalizedAge
      Scale Curve=&Lerp_Float(A=0.5, B=1, Alpha=$System.StateBlend)
     )
    )) {isSetParameters=true, module=SetVariables_#4}
   /SetVariables(Particles.StateTrailDynamic:Vector4f=&MakeVector4(
     X=$System.StateBlend
     Y=$System.StateBlendOpacity
     Z=&FloatFromCurve(
      FloatCurve=DataInterface() {propertyValues {Curve {
         defaultValue=0
         keys [{interpMode=RCIM_Cubic, tangentMode=RCTM_User, time=0, value=0}, {interpMode=RCIM_Cubic, tangentMode=RCTM_User, time=1, value=0.5}]
        }}}
      CurveIndex=$Particles.AdjustedNormalizedAge
     )
     W=0
    )) {isSetParameters=true, module=SetVariables_#5}
  ]
  rendererClasses [&NiagaraRibbonRendererProperties]
  renderers [
   {rendererClass=&NiagaraRibbonRendererProperties}
   {rendererClass=&NiagaraRibbonRendererProperties, rendererIndex=1}
  ]
 }
 Trail001 {
  bEnabled=false
  emitterSpawnScript [&FN_RibbonIDOffsetUtility()]
  emitterUpdateScript [
   &EmitterState()
   &FN_SpawnAnimTrail()
  ]
  particleSpawnScript [
   &FN_InitializeAnimTrailBoneIndecies()
   &FN_InitializeAnimTrail(Lifetime=0.18)
  ]
  particleUpdateScript [
   &SolveForcesAndVelocity(Clamp Velocity=false, Limit Acceleration=false)
   &FN_UpdateAnimTrail() {module=FN_UpdateAnimTrail001}
   &ScaleColor(
    Scale Alpha=&FloatFromCurve(
     FloatCurve=DataInterface() {propertyValues {Curve {keys [{interpMode=RCIM_Cubic, tangentMode=RCTM_User, time=0, value=1}, {interpMode=RCIM_Cubic, tangentMode=RCTM_User, time=1, value=0}]}}}
     Scale Curve=2
    )
    Scale RGB=&MakeVectorFromLinearColorRGB(Color=&ColorFromCurve(ColorCurve=DataInterface() {propertyValues {
        AlphaCurve {keys [<at 0 1 RCIM_Cubic>,<at 1 0 RCIM_Cubic>]}
        BlueCurve {keys [<at 0 0 RCIM_Cubic>,<at 0.7877 0>]}
        GreenCurve {keys [<at 0 1.6464 RCIM_Cubic>,<at 0.7877 0.5279>]}
        RedCurve {keys [<at 0 3 RCIM_Cubic>,<at 0.7877 1.5>]}
       }}))
    ScaleA=true
    ScaleRGB=true
   )
   /SetVariables(Particles.CleanTrailColor:LinearColor=&MakeLinearColorFromVectorAndFloat(
     "Vector (RGB)"=&VectorFromFloat(Value=&FloatFromCurve(
       FloatCurve=DataInterface() {propertyValues {Curve {defaultValue=1, keys [{interpMode=RCIM_Cubic, leaveTangent=-1.4397, tangentMode=RCTM_Break, time=0, value=1}]}}}
       Scale Curve=&Lerp_Float(A=0.75, B=1, Alpha=$System.StateBlend)
      ))
     "Float (Alpha)"=&FloatFromCurve(
      FloatCurve=DataInterface() {propertyValues {Curve {
         defaultValue=1
         keys [
          {arriveTangent=-0.0405, interpMode=RCIM_Cubic, leaveTangent=-0.2898, tangentMode=RCTM_Break, time=0, value=1}
          {arriveTangent=-2.2333, interpMode=RCIM_Cubic, leaveTangent=-2.2333, tangentMode=RCTM_User, time=1, value=0}
         ]
        }}}
      Scale Curve=&Lerp_Float(A=0.5, B=1, Alpha=$System.StateBlend)
     )
    )) {isSetParameters=true, module=SetVariables_#4}
   /SetVariables(Particles.StateTrailDynamic:Vector4f=&MakeVector4(
     X=$System.StateBlend
     Y=$System.StateBlendOpacity
     Z=&FloatFromCurve(FloatCurve=DataInterface() {propertyValues {Curve {
         defaultValue=0
         keys [{interpMode=RCIM_Cubic, tangentMode=RCTM_User, time=0, value=0}, {interpMode=RCIM_Cubic, tangentMode=RCTM_User, time=1, value=0.5}]
        }}})
     W=0
    )) {isSetParameters=true, module=SetVariables_#5}
  ]
  rendererClasses [&NiagaraRibbonRendererProperties]
  renderers [
   {rendererClass=&NiagaraRibbonRendererProperties}
   {rendererClass=&NiagaraRibbonRendererProperties, rendererIndex=1}
  ]
 }
}
summary {
 systemName=NS_Pickaxe_RainFNCS_Trail
 emitters [
  {emitterName=Trail, rendererClasses [&NiagaraRibbonRendererProperties]}
  {emitterName=GlitchShapes, rendererClasses [&NiagaraMeshRendererProperties, &NiagaraSpriteRendererProperties]}
  {bEnabled=false, emitterName=Trail001, rendererClasses [&NiagaraRibbonRendererProperties]}
 ]
 userVariables [
  User.bIsFrontEnd {type=NiagaraBool}
  User.bIsFrontEndPreview {type=NiagaraBool}
  User.FrontEndScale {type=NiagaraFloat, defaultValue=0.6, description=""}
  User.PhysicalSurface {type=NiagaraInt32}
  User.RibbonIDOffset {type=NiagaraInt32}
  User.StateBlend {type=NiagaraFloat, defaultValue=0, description=""}
  User.StateBlendOpacity {type=NiagaraFloat, defaultValue=0, description=""}
 ]
}
